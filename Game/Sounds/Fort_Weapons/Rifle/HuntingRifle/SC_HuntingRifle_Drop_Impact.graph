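# Sound cue for the hunting rifle's drop impact, judging by its name and its landed-sound waves.
# A random node picks one of three equally weighted landing waves and feeds a modulator at the root, with the cue pitched slightly down.
# It borrows the weapon reload attenuation and a low-ammo degraded ranged-weapon sound class.
# A concurrency override prevents new instances once the limit is reached.
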SoundCue /Game/Sounds/Fort_Weapons/Rifle/HuntingRifle/SC_HuntingRifle_Drop_Impact
root=Modulator_0
nodes [
 {id=Modulator_0, children [Random_0]}
 {id=Random_0, children [WavePlayer_0,WavePlayer_1,WavePlayer_2], properties {weights [1,1,1]}}
 {id=WavePlayer_0, properties {soundWaveAssetPtr=&/HuntingRifle_Drop_Landed_01}}
 {id=WavePlayer_1, properties {soundWaveAssetPtr=&/HuntingRifle_Drop_Landed_02}}
 {id=WavePlayer_2, properties {soundWaveAssetPtr=&/HuntingRifle_Drop_Landed_03}}
]
properties {
 attenuationSettings=&Fort_Weapon_Reload_Att
 bOverrideConcurrency=true
 maxDistance=2250
 pitchMultiplier=0.9
 soundClassObject=/Game/Sounds/FortSoundClassesAndModes/Fort_WeaponRanged_LowAmmo_Degraded_PSM
 volumeMultiplier=1
 concurrencyOverrides {resolutionRule=PreventNew}
}
