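# Sound cue for the hunting rifle's gunshot, built from close, medium and distant layers under ranged-weapon fire sound classes.
# Each layer has its own attenuation, volume modulation and random pick of three waves, feeding a mixer that weights close highest.
# The close layer also modulates pitch, blends in a sub hit borrowed from shotgun sounds, and uses the third-person class; the other two use the distant class.
# Local and global gunfire concurrency sets are assigned.
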SoundCue /Game/Sounds/Fort_Weapons/Rifle/HuntingRifle/SC_HuntingRifle_HT_Fire
root=Mixer_1
nodes [
 {id=Attenuation_0, children [Modulator_2], properties {attenuationSettings=&Fort_Weapon_Projectile_Close_Athena_Att}}
 {
  id=Attenuation_12
  children [Modulator_0]
  properties {attenuationSettings=&Fort_Weapon_Projectile_Medium_Mono_Athena_Att}
 }
 {
  id=Attenuation_13
  children [Modulator_1]
  properties {attenuationSettings=&Fort_Weapon_Projectile_Distant_Mono_Athena_Att}
 }
 {id=Mixer_0, children [WavePlayer_13, Random_3], properties {inputVolume [1,1]}}
 {id=Mixer_1, children [SoundClass_14, SoundClass_16, SoundClass_15], properties {inputVolume [0.8,0.25,0.2]}}
 {id=Modulator_0, children [Random_4], properties {volumeMax=2.2, volumeMin=2.1}}
 {id=Modulator_1, children [Random_5], properties {volumeMax=2.1, volumeMin=2}}
 {id=Modulator_2, children [Mixer_0], properties {pitchMax=1.1, pitchMin=1, volumeMax=1.4, volumeMin=1.3}}
 {id=Random_3, children [WavePlayer_10, WavePlayer_11, WavePlayer_12], properties {bShouldExcludeFromBranchCulling=true, weights [1,1,1]}}
 {id=Random_4, children [WavePlayer_8, WavePlayer_9, WavePlayer_14], properties {bShouldExcludeFromBranchCulling=true, weights [1,1,1]}}
 {id=Random_5, children [WavePlayer_15, WavePlayer_16, WavePlayer_17], properties {bShouldExcludeFromBranchCulling=true, weights [1,1,1]}}
 {
  id=SoundClass_14
  children [Attenuation_0]
  properties {soundClassOverride=&Fort_WeaponRanged_Fire_3P_PSM}
 }
 {
  id=SoundClass_15
  children [Attenuation_13]
  properties {soundClassOverride=&Fort_WeaponRanged_Fire_Distant}
 }
 {
  id=SoundClass_16
  children [Attenuation_12]
  properties {soundClassOverride=&Fort_WeaponRanged_Fire_Distant}
 }
 {id=WavePlayer_10, properties {soundWaveAssetPtr=&/HuntingRifle_Shot_Player_01_v2}}
 {id=WavePlayer_11, properties {soundWaveAssetPtr=&/HuntingRifle_Shot_Player_02_v2}}
 {id=WavePlayer_12, properties {soundWaveAssetPtr=&/HuntingRifle_Shot_Player_03_v2}}
 {id=WavePlayer_13, properties {soundWaveAssetPtr=/Game/Sounds/Fort_Weapons/Shotgun/Charge/Sub_Hit_N445}}
 {id=WavePlayer_14, properties {soundWaveAssetPtr=&/HuntingRifle_Shot_Player_Medium_01}}
 {id=WavePlayer_15, properties {soundWaveAssetPtr=&/HuntingRifle_Shot_Player_Distant_03}}
 {id=WavePlayer_16, properties {soundWaveAssetPtr=&/HuntingRifle_Shot_Player_Distant_02}}
 {id=WavePlayer_17, properties {soundWaveAssetPtr=&/HuntingRifle_Shot_Player_Distant_01}}
 {id=WavePlayer_8, properties {soundWaveAssetPtr=&/HuntingRifle_Shot_Player_Medium_03}}
 {id=WavePlayer_9, properties {soundWaveAssetPtr=&/HuntingRifle_Shot_Player_Medium_02}}
]
properties {
 maxDistance=27500
 soundClassObject=&Fort_WeaponRanged_Fire_3P_PSM
 volumeMultiplier=1
 attenuationOverrides {distanceAlgorithm=NaturalSound, falloffDistance=4000, attenuationShapeExtents {x=1000}}
 concurrencyOverrides {bLimitToOwner=true, maxCount=3, resolutionRule=StopOldest}
 concurrencySet [&Weapon_Gunfire_Local_Conc, &Weapon_Gunfire_Global_Conc]
}
