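# Sound cue for the first-person hunting rifle fire sound.
# A random node picks one of three player shot variations, which is mixed with a sub hit layer borrowed from the Charge shotgun sounds.
# A root modulator slightly varies pitch and boosts volume on each play.
# It uses first-person weapon attenuation, sound class and gunfire concurrency, carries per-owner stop-oldest override settings, and has its debug flag on.
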SoundCue /Game/Sounds/Fort_Weapons/Rifle/HuntingRifle/SC_HuntingRifle_HT_Fire_1P
root=Modulator_2
nodes [
 {id=Mixer_0, children [WavePlayer_13, Random_0], properties {inputVolume [1,1]}}
 {id=Modulator_2, children [Mixer_0], properties {pitchMax=1.1, pitchMin=1, volumeMax=2, volumeMin=1.8}}
 {id=Random_0, children [WavePlayer_10, WavePlayer_11, WavePlayer_12], properties {bShouldExcludeFromBranchCulling=true, weights [1,1,1]}}
 {id=WavePlayer_10, properties {soundWaveAssetPtr=&/HuntingRifle_Shot_Player_01_v2}}
 {id=WavePlayer_11, properties {soundWaveAssetPtr=&/HuntingRifle_Shot_Player_02_v2}}
 {id=WavePlayer_12, properties {soundWaveAssetPtr=&/HuntingRifle_Shot_Player_03_v2}}
 {id=WavePlayer_13, properties {soundWaveAssetPtr=/Game/Sounds/Fort_Weapons/Shotgun/Charge/Sub_Hit_N445}}
]
properties {
 attenuationSettings=&Fort_Weapon_Projectile_FirstPerson_Athena_Att
 bDebug=true
 maxDistance=2097152
 soundClassObject=&Fort_WeaponRanged_Fire_1P_PSM
 volumeMultiplier=0.8
 concurrencyOverrides {bLimitToOwner=true, maxCount=3, resolutionRule=StopOldest}
 concurrencySet [&Weapon_Gunfire_FirstPerson_Conc]
}
